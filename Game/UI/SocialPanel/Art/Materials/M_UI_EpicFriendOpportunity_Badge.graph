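# Translucent UI material for the Epic friend opportunity badge in the social panel, drawing a corner triangle with a plus sign cut out.
# Two DrawLine-2D calls form the plus from texture coordinates, with parameterized thickness; scale and shift parameters build adjusted UVs alongside one customized UV slot.
# The inverted plus multiplies a sharp diagonal mask, made from normalized UI U minus V, and tints it white, all without textures.
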Material /Game/UI/SocialPanel/Art/Materials/M_UI_EpicFriendOpportunity_Badge
graph {
 properties {
  bUsedWithStaticMesh=false
  blendMode=BLEND_Translucent
  materialDomain=MD_UI
  numCustomizedUVs=1
  pixelDepthOffsetMode=PDOM_Legacy
  usedShadingModels=Default Lit
 }
 nodes [
  Constant()
  n1 = ScalarParameter() {defaultValue=0.5, parameterName=PlusSign_Scale}
  n2 = MaterialFunctionCall(Float2=MaterialFunctionCall(Texture Scale=MaterialFunctionCall(X=@n1, Y=@n1) {materialFunction=&MakeFloat2}.Result) {materialFunction=&ScaleUVsByCenter}.UVs) {materialFunction=&BreakOutFloat2Components}
  MaterialFunctionCall(X=Add(A=@n2.R, B=ScalarParameter() {parameterName=PlusSign_XShift}), Y=Add(A=@n2.G, B=ScalarParameter() {parameterName=PlusSign_YShift})) {materialFunction=&MakeFloat2}
  n3 = TextureCoordinate()
  n4 = Constant2Vector() {r=0.5}
  n5 = Constant2Vector() {g=1, r=0.5}
  n6 = ScalarParameter() {defaultValue=0.025, parameterName=PlusSign_Thickness}
  n7 = StaticBool() {value=true}
  n8 = Constant2Vector() {g=0.5}
  n9 = Constant2Vector() {g=0.5, r=1}
  n10 = Clamp(None=Add(
    A=MaterialFunctionCall!(Coordinates=@n3, P0=@n4, P1=@n5, Line Width=@n6, Edge Falloff, Square Corners?=@n7) {materialFunction=/Engine/Functions/Engine_MaterialFunctions01/Coordinates/DrawLine-2D}.Result
    B=MaterialFunctionCall(Coordinates=@n3, P0=@n8, P1=@n9, Line Width=@n6, Square Corners?=@n7) {materialFunction=&DrawLine-2D}.Result
   ))
  n11 = MaterialFunctionCall() {materialFunction=&GetUserInterfaceUV}
  n12 = Multiply(A=Subtract(A=ComponentMask(None=@n11.Normalized UV) {r=true}, B=ComponentMask(None=@n11.Normalized UV) {g=true})) {constB=100}
  Multiply(A=Constant3Vector() {constant=<rgba 1 1 1 0>}, B=Multiply(A=OneMinus(None=@n10), B=Clamp(None=@n12)))
 ]
}
