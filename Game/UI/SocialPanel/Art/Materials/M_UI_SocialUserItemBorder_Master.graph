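# Translucent UI master material that colors the border of user items in the social panel.
# The horizontal normalized UI UV samples two curve atlas gradient parameters, a base and a highlight, which share the same default curve.
# GradientFill blends a solid color toward the base gradient, then AnimateHighlight blends that result toward the highlight gradient.
# A separate PrimaryAlpha scalar, not driven by the gradients, provides the alpha.
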Material /Game/UI/SocialPanel/Art/Materials/M_UI_SocialUserItemBorder_Master
graph {
 properties {
  bUsedWithStaticMesh=false
  blendMode=BLEND_Translucent
  materialDomain=MD_UI
  pixelDepthOffsetMode=PDOM_Legacy
  usedShadingModels=Default Lit
 }
 nodes [
  n1 = ComponentMask(None=MaterialFunctionCall() {materialFunction=&GetUserInterfaceUV}.Normalized UV) {r=true}
  n2 = CurveAtlasRowParameter(CurveTime=@n1) {atlas=/Game/UI/SocialPanel/Art/CA_SocialPanel, curve=/Game/UI/SocialPanel/Art/CC_SocialPanel_UserListBaclground, parameterName=GradientCurve}
  n3 = CurveAtlasRowParameter(CurveTime=@n1) {
   atlas=&CA_SocialPanel
   curve=&CC_SocialPanel_UserListBaclground
   parameterName=HighlightGradientCurve
  }
  LinearInterpolate(
   A=LinearInterpolate(
    A=VectorParameter() {parameterName=RGB_SolidFill, defaultValue=<rgba 0 0.0802 0.2664 1>}.RGB
    B=@n2
    Alpha=ScalarParameter() {parameterName=GradientFill}
   )
   B=@n3
   Alpha=ScalarParameter() {parameterName=AnimateHighlight, sliderMax=1}
  )
  ScalarParameter() {defaultValue=1, parameterName=PrimaryAlpha}
 ]
}
