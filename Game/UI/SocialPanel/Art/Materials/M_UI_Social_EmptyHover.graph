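# Translucent UI material for the social panel, apparently a hover state for empty entries going by its name.
# It builds a horizontal mask from the U texture coordinate, solid up to three quarters across, with a steep, narrow falloff to zero after that.
# The clamped mask interpolates toward the alpha of the RGBA vector parameter, whose default color is red.
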Material /Game/UI/SocialPanel/Art/Materials/M_UI_Social_EmptyHover
graph {
 properties {
  bUsedWithStaticMesh=false
  blendMode=BLEND_Translucent
  materialDomain=MD_UI
  pixelDepthOffsetMode=PDOM_Legacy
  usedShadingModels=Default Lit
 }
 nodes [
  n1 = TextureCoordinate()
  n2 = Multiply(A=ComponentMask(None=Add(A=@n1) {constB=-0.75}) {r=true}) {constB=128}
  LinearInterpolate(B=VectorParameter() {parameterName=RGBA, defaultValue=<rgba 1 0 0 1>}.A, Alpha=Clamp(None=OneMinus(None=@n2)))
 ]
}
