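# UI master material for social panel icons, which tints a channel-packed mask texture with parameterized colors, using a social invite icon by default.
# The red channel masks the main fill color, while blue multiplied by inverted red masks the shadow color so it stays outside the fill.
# The two tinted layers are summed for color, and a separate sum combines red with the shadow mask scaled by ShadowOpacity.
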Material /Game/UI/SocialPanel/Art/Materials/M_UI_Social_MaskedIcon_Master
graph {
 properties {
  bUsedWithStaticMesh=false
  blendMode=BLEND_Translucent
  materialDomain=MD_UI
  pixelDepthOffsetMode=PDOM_Legacy
  usedShadingModels=Default Lit
 }
 nodes [
  n1 = TextureSampleParameter2D() {parameterName=MaskedIconTexture, texture=/Game/UI/SocialPanel/Art/Textures/T_UI_Icon_Masked_Social_Invite}
  n2 = Multiply(A=OneMinus(None=@n1.R), B=@n1.B)
  Add(
   A=Multiply(A=VectorParameter() {parameterName=MainFillColor, defaultValue=<rgba 1 1 1 1>}.RGB, B=@n1.R)
   B=Multiply(A=VectorParameter() {parameterName=ShadowColor, defaultValue=<rgba 0.0885 0.3897 0.9271 1>}.RGB, B=@n2)
  )
  Add(A=@n1.R, B=Multiply(A=@n2, B=ScalarParameter() {defaultValue=1, parameterName=ShadowOpacity}))
  Clamp(None=Add(A=@n1.R, B=@n1.R))
  Multiply()
 ]
}
