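# Translucent UI material for the social panel's user hover, producing a hard, angled wipe mask with a bolt-like jog, alongside an RGB color parameter.
# Two rotated-UV edges, offset by BoltOffset and sharpened by multiply-and-clamp, blend across a smoothstepped, angled center split when UseBoltMask is on.
# AnimateStateChange slides the edges between start and end offsets.
# A FlipDirection static switch negates angles and offsets and inverts the mask.
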Material /Game/UI/SocialPanel/Art/Materials/M_UI_UserHover
graph {
 properties {
  bUsedWithStaticMesh=false
  blendMode=BLEND_Translucent
  materialDomain=MD_UI
  pixelDepthOffsetMode=PDOM_Legacy
  usedShadingModels=Default Lit
 }
 nodes [
  n1 = ScalarParameter() {defaultValue=-20, group="Angles(Degrees)", parameterName=CenterAngle}
  n2 = Multiply(A=@n1) {constB=-1}
  n3 = TextureCoordinate()
  n4 = Rotator(Coordinate=@n3, Time=Divide(A=StaticSwitchParameter(True=@n2, False=@n1) {parameterName=FlipDirection}) {constB=360}) {speed=6.2832}
  n5 = ScalarParameter() {group=Offsets, parameterName=CenterOffset}
  n6 = ComponentMask(None=Add(A=Add(A=@n4) {constB=-0.5}, B=Multiply(A=@n5) {constB=-1})) {g=true}
  n7 = ScalarParameter() {defaultValue=0.0125, parameterName=CenterEdgeSS}
  Multiply(A=@n6, B=@n7)
  n8 = ScalarParameter() {defaultValue=10, group="Angles(Degrees)", parameterName=EdgeAngle}
  n9 = Divide(A=StaticSwitchParameter(True=Multiply(A=@n8) {constB=-1}, False=@n8) {parameterName=FlipDirection}) {constB=360}
  n10 = ScalarParameter() {defaultValue=0.125, group=Offsets, parameterName=BoltOffset}
  n11 = Add(A=@n10) {constB=0}
  n12 = Constant()
  n13 = Add(A=@n12) {constB=-1}
  n14 = StaticSwitchParameter(True=Constant(), False=Multiply(A=@n10) {constB=-1}) {parameterName=FlipDirection}
  n15 = Add(
   A=Add(A=Rotator(Coordinate=@n3, Time=@n9) {speed=6.2832}, B=StaticSwitchParameter(True=@n11, False=@n13) {parameterName=FlipDirection})
   B=@n14
  )
  n16 = ScalarParameter() {parameterName=AnimStartOffset}
  n17 = ScalarParameter() {defaultValue=0.5, parameterName=AnimEndOffset}
  n18 = LinearInterpolate(
   A=StaticSwitchParameter(True=Multiply(A=@n16) {constB=-1}, False=@n16) {parameterName=FlipDirection}
   B=StaticSwitchParameter(True=Multiply(A=@n17) {constB=-1}, False=@n17) {parameterName=FlipDirection}
   Alpha=ScalarParameter() {parameterName=AnimateStateChange, sliderMax=1}
  )
  n19 = ScalarParameter() {defaultValue=512, parameterName=PrimaryEdgeSharpness}
  n20 = Multiply(A=ComponentMask(None=Add(A=@n15, B=@n18)) {r=true}, B=@n19)
  n21 = Add(A=@n12) {constB=0}
  n22 = Add(A=@n10) {constB=-1}
  n23 = Add(
   A=Add(A=Rotator(Coordinate=@n3, Time=@n9) {speed=6.2832}, B=StaticSwitchParameter(True=@n21, False=@n22) {parameterName=FlipDirection})
   B=@n14
  )
  n24 = Multiply(A=ComponentMask(None=Add(A=@n23, B=@n18)) {r=true}, B=@n19)
  n25 = Constant() {r=0.5}
  n26 = Multiply(A=@n7) {constB=-1}
  n27 = Clamp(None=MaterialFunctionCall(Alpha=@n6, Min=Add(A=@n25, B=@n26), Max=Add(A=@n25, B=@n7)) {materialFunction=&SmoothStep}.Result)
  n28 = Clamp(None=StaticSwitchParameter(True=LinearInterpolate(A=@n20, B=@n24, Alpha=@n27), False=@n20) {defaultValue=true, parameterName=UseBoltMask})
  StaticSwitchParameter(True=OneMinus(None=@n28), False=@n28) {parameterName=FlipDirection}
  VectorParameter() {parameterName=RGB, defaultValue=<rgba 0.0723 0.8714 1 1>}
 ]
}
